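# Physics asset for the Vertigo Tongue gadget's skeletal mesh, giving the tongue a jointed collision and physics setup.
# Seven capsule bodies follow bones tongue_1 through tongue_7, generally tapering in radius from the first to the last.
# Six constraints link each segment to the one before it in a single chain, all sharing the same symmetric swing and twist angle limits for bounded bending.
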PhysicsAsset /Game/Gadgets/Assets/Vertigo_Tongue/Meshes/M_LRG_Vertigo_Tongue_Gadget_Physics
bodies {
 tongue_1 [{shapeType=Capsule, length=3.5575, radius=2.9883, center=<2.1353 0.3069 0.0001>, rotation=<rot -0.0002 -90 -89.9982>}]
 tongue_2 [{shapeType=Capsule, length=3, radius=2.4498, center=<3.623 -0.1393 0>, rotation=<rot -0.0748 -86.9733 91.4156>}]
 tongue_3 [{shapeType=Capsule, length=4.6491, radius=2.2232, center=<3.9751 0.1909 0>, rotation=<rot 0 91.8077 -90>}]
 tongue_4 [{shapeType=Capsule, length=3.75, radius=2.0303, center=<3.3274 0.1114 0>, rotation=<rot 0 -89.9997 82.315>}]
 tongue_5 [{shapeType=Capsule, length=1.6403, radius=1.5055, center=<1.1825 0.1186 0>, rotation=<rot 0 90.3615 -90>}]
 tongue_6 [{shapeType=Capsule, length=1.8899, radius=1.4003, center=<0.5327 0.1005 0>, rotation=<rot 0 -89.5044 -90>}]
 tongue_7 [{shapeType=Capsule, length=1.005, radius=0.505, center=<1.5253 0.0685 0>, rotation=<rot 0 -90.6058 -90>}]
}
constraints [
 {bone1Name=tongue_2, bone2Name=tongue_1, swing1LimitDegrees=45, swing2LimitDegrees=45, twistLimitDegrees=45}
 {bone1Name=tongue_3, bone2Name=tongue_2, swing1LimitDegrees=45, swing2LimitDegrees=45, twistLimitDegrees=45}
 {bone1Name=tongue_4, bone2Name=tongue_3, swing1LimitDegrees=45, swing2LimitDegrees=45, twistLimitDegrees=45}
 {bone1Name=tongue_5, bone2Name=tongue_4, swing1LimitDegrees=45, swing2LimitDegrees=45, twistLimitDegrees=45}
 {bone1Name=tongue_6, bone2Name=tongue_5, swing1LimitDegrees=45, swing2LimitDegrees=45, twistLimitDegrees=45}
 {bone1Name=tongue_7, bone2Name=tongue_6, swing1LimitDegrees=45, swing2LimitDegrees=45, twistLimitDegrees=45}
]
